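# Physics asset for the Progressive Jonesy Kepler medium male player body, setting up secondary motion for shirt flaps, belt buckles and pouches.
# Capsules cover spine, pelvis, arms, clavicles and right thigh, while dangling pieces use capsules or boxes.
# Each dyn bone hangs from a tiny MSTR anchor body with free angular motion on the pelvis or, for pouches, spine_02.
# Angular limits constrain each piece against its anchor, tightest on the pouches.
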PhysicsAsset /BRCosmetics/Characters/Player/Male/Medium/Bodies/M_MED_Progressive_Jonesy_Kepler/Meshes/M_MED_Progressive_Jonesy_kepler_Physics
bodies {
 clavicle_l [{shapeType=Capsule, length=9.7203, radius=5.7303, center=<8.4231 -0.5029 1.5882>, rotation=<rot -90 0.0197 -0.0197>}]
 clavicle_r [{shapeType=Capsule, length=10.2882, radius=6, center=<-9.2103 0 -1.5882>, rotation=<rot -90 -0.0017 -179.9982>}]
 dyn_beltbuckle_01_l [{shapeType=Capsule, length=5.269, radius=2.3814, center=<2.494 0 0.0002>, rotation=<rot 0.0013 89.9994 89.9999>}]
 dyn_beltbuckle_01_l_MSTR [{shapeType=Capsule, length=0.01, radius=0.01}]
 dyn_cb_beltbuckle_01 [{shapeType=Capsule, length=6.3929, radius=2.0655, center=<5.1448 0 0>, rotation=<rot 0 90 90>}]
 dyn_cb_buckle_01_MSTR [{shapeType=Capsule, length=0.1, radius=0.1, rotation=<rot 0 -90 0>}]
 dyn_pouch_bk_01 [{shapeType=Capsule, length=6.1961, radius=2.3625, center=<1.8257 0 0>}]
 dyn_pouch_bk_02 [{shapeType=Capsule, length=6.1904, radius=2.409, center=<1.863 0 0>}]
 dyn_pouch_bk_03 [{shapeType=Capsule, length=6.2144, radius=2.3582, center=<1.8449 0 0>}]
 dyn_pouch_bk_MSTR_01 [{shapeType=Capsule, length=0.505, radius=0.505, rotation=<rot 0 -90 0>}]
 dyn_pouch_bk_MSTR_02 [{shapeType=Capsule, length=0.505, radius=0.505, rotation=<rot 0 -90 0>}]
 dyn_pouch_bk_MSTR_03 [{shapeType=Capsule, length=0.505, radius=0.505, rotation=<rot 0 -90 0>}]
 dyn_pouch_lt_01 [{shapeType=Capsule, length=6.1755, radius=2.4015, center=<1.8222 0 0>}]
 dyn_pouch_lt_02 [{shapeType=Capsule, length=6.1985, radius=2.4062, center=<1.8198 0 0>}]
 dyn_pouch_lt_03 [{shapeType=Capsule, length=6.2029, radius=2.3595, center=<1.7815 0 0>}]
 dyn_pouch_lt_MSTR_01 [{shapeType=Capsule, length=0.505, radius=0.505, rotation=<rot 0 -90 0>}]
 dyn_pouch_lt_MSTR_02 [{shapeType=Capsule, length=0.505, radius=0.505, rotation=<rot 0 -90 0>}]
 dyn_pouch_lt_MSTR_03 [{shapeType=Capsule, length=0.505, radius=0.505, rotation=<rot 0 -90 0>}]
 dyn_pouch_rt_01 [{shapeType=Capsule, length=6.1602, radius=2.4297, center=<1.7925 0 0>}]
 dyn_pouch_rt_02 [{shapeType=Capsule, length=6.1997, radius=2.4092, center=<1.8479 0 0>}]
 dyn_pouch_rt_03 [{shapeType=Capsule, length=6.2027, radius=2.3646, center=<1.8303 0 0>}]
 dyn_pouch_rt_MSTR_01 [{shapeType=Capsule, length=0.505, radius=0.505, rotation=<rot 0 -90 0>}]
 dyn_pouch_rt_MSTR_02 [{shapeType=Capsule, length=0.505, radius=0.505, rotation=<rot 0 -90 0>}]
 dyn_pouch_rt_MSTR_03 [{shapeType=Capsule, length=0.505, radius=0.505, rotation=<rot 0 -90 0>}]
 dyn_shirt_01_r [{shapeType=Box, extentX=12.637, extentY=13.0328, extentZ=4.2891, center=<6.3255 0 0>}]
 dyn_shirt_01_r_MSTR [{shapeType=Capsule, length=0.01, radius=0.01}]
 dyn_shirt_bk_01 [{shapeType=Capsule, length=4.7971, radius=2.8533, center=<3.2085 0 0>, rotation=<rot 0 90 90>}]
 dyn_shirt_bk_01_MSTR [{shapeType=Capsule, length=0.01, radius=0.01}]
 dyn_shirt_bk_r_01 [{shapeType=Capsule, length=7.8695, radius=3.7573, center=<8 0 0>, rotation=<rot 0 90 90>}]
 dyn_shirt_bk_r_01_MSTR [{shapeType=Capsule, length=0.505, radius=0.505, rotation=<rot 0 -90 0>}]
 dyn_shirt_fr_01_r [{shapeType=Box, extentX=11.9082, extentY=2.9505, extentZ=13.7703, center=<6.5541 0.3984 0.4749>, rotation=<rot 0 0 93.6014>}]
 dyn_shirt_fr_01_r_MSTR [{shapeType=Capsule, length=0.01, radius=0.01}]
 lowerarm_l [{shapeType=Capsule, length=17.5, radius=4.75, center=<11.8488 0 0>, rotation=<rot -90 0.0264 -0.0263>}]
 lowerarm_r [{shapeType=Capsule, length=17.5, radius=4.75, center=<-11.8488 0 0>, rotation=<rot -90 -179.9802 -0.0197>}]
 pelvis [{shapeType=Capsule, length=11.5, radius=10.75}]
 spine_02 [{shapeType=Capsule, length=11.5, radius=10.75}]
 spine_04 [{shapeType=Capsule, length=11.5, radius=10.75}]
 spine_05 [{shapeType=Capsule, length=12.5, radius=9.75}]
 thigh_r [{shapeType=Capsule, length=33.5, radius=6.75, center=<19.8348 0 0>, rotation=<rot -90 -0.0148 0.0148>}]
 upperarm_l [{shapeType=Capsule, length=15.5, radius=6.75, center=<11.8488 0 0>, rotation=<rot -90 0.0264 -0.0263>}]
 upperarm_r [{shapeType=Capsule, length=15.5, radius=6.75, center=<-11.8488 0 0>, rotation=<rot -90 -179.9802 -0.0197>}]
}
constraints [
 {bone1Name=spine_02, bone2Name=pelvis, swing1LimitDegrees=45, swing2LimitDegrees=45, twistLimitDegrees=45}
 {bone1Name=spine_04, bone2Name=spine_02, swing1LimitDegrees=45, swing2LimitDegrees=45, twistLimitDegrees=45}
 {bone1Name=spine_05, bone2Name=spine_04, swing1LimitDegrees=45, swing2LimitDegrees=45, twistLimitDegrees=45}
 {
  bone1Name=upperarm_l
  bone2Name=clavicle_l
  swing1LimitDegrees=45
  swing1Motion=Free
  swing2LimitDegrees=45
  swing2Motion=Free
  twistLimitDegrees=45
  twistMotion=Free
 }
 {
  bone1Name=upperarm_r
  bone2Name=clavicle_r
  swing1LimitDegrees=45
  swing1Motion=Free
  swing2LimitDegrees=45
  swing2Motion=Free
  twistLimitDegrees=45
  twistMotion=Free
 }
 {
  bone1Name=dyn_shirt_bk_01_MSTR
  bone2Name=pelvis
  swing1LimitDegrees=45
  swing1Motion=Free
  swing2LimitDegrees=45
  swing2Motion=Free
  twistLimitDegrees=45
  twistMotion=Free
 }
 {
  bone1Name=dyn_beltbuckle_01_l_MSTR
  bone2Name=pelvis
  swing1LimitDegrees=45
  swing1Motion=Free
  swing2LimitDegrees=45
  swing2Motion=Free
  twistLimitDegrees=45
  twistMotion=Free
 }
 {
  bone1Name=dyn_shirt_fr_01_r_MSTR
  bone2Name=pelvis
  swing1LimitDegrees=45
  swing1Motion=Free
  swing2LimitDegrees=45
  swing2Motion=Free
  twistLimitDegrees=45
  twistMotion=Free
 }
 {bone1Name=dyn_shirt_bk_01, bone2Name=dyn_shirt_bk_01_MSTR, swing1LimitDegrees=20, swing2LimitDegrees=45, twistLimitDegrees=5}
 {bone1Name=dyn_beltbuckle_01_l, bone2Name=dyn_beltbuckle_01_l_MSTR, swing1LimitDegrees=45, swing2LimitDegrees=45, twistLimitDegrees=5}
 {bone1Name=dyn_shirt_fr_01_r, bone2Name=dyn_shirt_fr_01_r_MSTR, swing1LimitDegrees=15, swing2LimitDegrees=45, twistLimitDegrees=5}
 {
  bone1Name=lowerarm_l
  bone2Name=upperarm_l
  swing1LimitDegrees=45
  swing1Motion=Free
  swing2LimitDegrees=45
  swing2Motion=Free
  twistLimitDegrees=45
  twistMotion=Free
 }
 {
  bone1Name=lowerarm_r
  bone2Name=upperarm_r
  swing1LimitDegrees=45
  swing1Motion=Free
  swing2LimitDegrees=45
  swing2Motion=Free
  twistLimitDegrees=45
  twistMotion=Free
 }
 {
  bone1Name=thigh_r
  bone2Name=pelvis
  swing1LimitDegrees=45
  swing1Motion=Free
  swing2LimitDegrees=45
  swing2Motion=Free
  twistLimitDegrees=45
  twistMotion=Free
 }
 {
  bone1Name=dyn_shirt_01_r_MSTR
  bone2Name=pelvis
  swing1LimitDegrees=45
  swing1Motion=Free
  swing2LimitDegrees=45
  swing2Motion=Free
  twistLimitDegrees=45
  twistMotion=Free
 }
 {
  bone1Name=clavicle_l
  bone2Name=spine_05
  swing1LimitDegrees=45
  swing1Motion=Free
  swing2LimitDegrees=45
  swing2Motion=Free
  twistLimitDegrees=45
  twistMotion=Free
 }
 {
  bone1Name=clavicle_r
  bone2Name=spine_05
  swing1LimitDegrees=45
  swing1Motion=Free
  swing2LimitDegrees=45
  swing2Motion=Free
  twistLimitDegrees=45
  twistMotion=Free
 }
 {bone1Name=dyn_shirt_bk_r_01, bone2Name=dyn_shirt_bk_r_01_MSTR, swing1LimitDegrees=20, swing2LimitDegrees=45, twistLimitDegrees=5}
 {
  bone1Name=dyn_shirt_bk_r_01_MSTR
  bone2Name=pelvis
  swing1LimitDegrees=45
  swing1Motion=Free
  swing2LimitDegrees=45
  swing2Motion=Free
  twistLimitDegrees=45
  twistMotion=Free
 }
 {bone1Name=dyn_shirt_01_r, bone2Name=dyn_shirt_01_r_MSTR, swing1LimitDegrees=20, swing2LimitDegrees=45, twistLimitDegrees=5}
 {
  bone1Name=dyn_cb_buckle_01_MSTR
  bone2Name=pelvis
  swing1LimitDegrees=45
  swing1Motion=Free
  swing2LimitDegrees=45
  swing2Motion=Free
  twistLimitDegrees=45
  twistMotion=Free
 }
 {bone1Name=dyn_cb_beltbuckle_01, bone2Name=dyn_cb_buckle_01_MSTR, swing1LimitDegrees=45, swing2LimitDegrees=45, twistLimitDegrees=5}
 {bone1Name=dyn_pouch_rt_03, bone2Name=dyn_pouch_rt_MSTR_03, swing1LimitDegrees=3, swing2LimitDegrees=7, twistLimitDegrees=3}
 {bone1Name=dyn_pouch_rt_02, bone2Name=dyn_pouch_rt_MSTR_02, swing1LimitDegrees=3, swing2LimitDegrees=7, twistLimitDegrees=3}
 {bone1Name=dyn_pouch_rt_01, bone2Name=dyn_pouch_rt_MSTR_01, swing1LimitDegrees=3, swing2LimitDegrees=7, twistLimitDegrees=3}
 {bone1Name=dyn_pouch_lt_01, bone2Name=dyn_pouch_lt_MSTR_01, swing1LimitDegrees=3, swing2LimitDegrees=7, twistLimitDegrees=3}
 {bone1Name=dyn_pouch_lt_02, bone2Name=dyn_pouch_lt_MSTR_02, swing1LimitDegrees=3, swing2LimitDegrees=7, twistLimitDegrees=3}
 {bone1Name=dyn_pouch_lt_03, bone2Name=dyn_pouch_lt_MSTR_03, swing1LimitDegrees=3, swing2LimitDegrees=7, twistLimitDegrees=3}
 {bone1Name=dyn_pouch_bk_03, bone2Name=dyn_pouch_bk_MSTR_03, swing1LimitDegrees=3, swing2LimitDegrees=7, twistLimitDegrees=3}
 {bone1Name=dyn_pouch_bk_02, bone2Name=dyn_pouch_bk_MSTR_02, swing1LimitDegrees=3, swing2LimitDegrees=7, twistLimitDegrees=3}
 {bone1Name=dyn_pouch_bk_01, bone2Name=dyn_pouch_bk_MSTR_01, swing1LimitDegrees=3, swing2LimitDegrees=7, twistLimitDegrees=3}
 {
  bone1Name=dyn_pouch_rt_MSTR_03
  bone2Name=spine_02
  swing1LimitDegrees=45
  swing1Motion=Free
  swing2LimitDegrees=45
  swing2Motion=Free
  twistLimitDegrees=45
  twistMotion=Free
 }
 {
  bone1Name=dyn_pouch_rt_MSTR_02
  bone2Name=spine_02
  swing1LimitDegrees=45
  swing1Motion=Free
  swing2LimitDegrees=45
  swing2Motion=Free
  twistLimitDegrees=45
  twistMotion=Free
 }
 {
  bone1Name=dyn_pouch_rt_MSTR_01
  bone2Name=spine_02
  swing1LimitDegrees=45
  swing1Motion=Free
  swing2LimitDegrees=45
  swing2Motion=Free
  twistLimitDegrees=45
  twistMotion=Free
 }
 {
  bone1Name=dyn_pouch_lt_MSTR_01
  bone2Name=spine_02
  swing1LimitDegrees=45
  swing1Motion=Free
  swing2LimitDegrees=45
  swing2Motion=Free
  twistLimitDegrees=45
  twistMotion=Free
 }
 {
  bone1Name=dyn_pouch_lt_MSTR_02
  bone2Name=spine_02
  swing1LimitDegrees=45
  swing1Motion=Free
  swing2LimitDegrees=45
  swing2Motion=Free
  twistLimitDegrees=45
  twistMotion=Free
 }
 {
  bone1Name=dyn_pouch_lt_MSTR_03
  bone2Name=spine_02
  swing1LimitDegrees=45
  swing1Motion=Free
  swing2LimitDegrees=45
  swing2Motion=Free
  twistLimitDegrees=45
  twistMotion=Free
 }
 {
  bone1Name=dyn_pouch_bk_MSTR_03
  bone2Name=spine_02
  swing1LimitDegrees=45
  swing1Motion=Free
  swing2LimitDegrees=45
  swing2Motion=Free
  twistLimitDegrees=45
  twistMotion=Free
 }
 {
  bone1Name=dyn_pouch_bk_MSTR_02
  bone2Name=spine_02
  swing1LimitDegrees=45
  swing1Motion=Free
  swing2LimitDegrees=45
  swing2Motion=Free
  twistLimitDegrees=45
  twistMotion=Free
 }
 {
  bone1Name=dyn_pouch_bk_MSTR_01
  bone2Name=spine_02
  swing1LimitDegrees=45
  swing1Motion=Free
  swing2LimitDegrees=45
  swing2Motion=Free
  twistLimitDegrees=45
  twistMotion=Free
 }
]
