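# Animation Blueprint for the Progressive Jonesy Kepler body, adding secondary motion to shoulder pads, pouches, crossbow and shirt.
# A copied parent pose passes through two Control Rigs (second off while skydiving), state-driven bone offsets, weighted constraints and two RigidBody simulations.
# Both simulations take gravity from windVector, which becomes scaled skydiving force during skydiving.
# The event graph also derives a weapon-dependent right-arm piston alpha, left unconnected in the AnimGraph.
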AnimBlueprint /BRCosmetics/Characters/Player/Male/Medium/Bodies/M_MED_Progressive_Jonesy_Kepler/Meshes/M_MED_Progressive_Jonesy_Kepler_AnimBP
target_skeleton=/Game/Characters/Player/Male/Medium/Bodies/M_MED_Progressive_Jonesy_Kepler/Meshes/M_MED_Progressive_Jonesy_Kepler_Skeleton
graphs [
 AnimGraph {nodes [
   "Transform (Modify) Bone - Bone: dyn_shoulderpad_r_base"(Alpha=1, Translation="1,-0.25,0", Rotation="3,0,0", Scale="1,1,1") {
    rotationMode=BMM_Additive
    rotationSpace=BCS_BoneSpace
    translationMode=BMM_Additive
    translationSpace=BCS_BoneSpace
    alphaBoolBlend {blendOption=Linear}
    alphaScaleBiasClamp {interpSpeedIncreasing=40}
    boneToModify {boneName=dyn_shoulderpad_r_base}
    rotation {pitch=3}
    translation=<1 -0.25>
   }
   "Transform (Modify) Bone - Bone: dyn_shoulderpad_MSTR"(Alpha=1, Translation="0,0,0", Rotation="0,0,0", Scale="1,1,1") {
    rotationMode=BMM_Additive
    rotationSpace=BCS_BoneSpace
    translationMode=BMM_Additive
    translationSpace=BCS_BoneSpace
    alphaBoolBlend {blendOption=Linear}
    alphaScaleBiasClamp {interpSpeedIncreasing=40}
    boneToModify {boneName=dyn_shoulderpad_MSTR}
   }
   n1 = Control Rig!(Source=Copy Pose From Mesh() {bUseAttachedParent=true}, bAlphaBoolEnabled:bool=True) -> (Pose) {
    alphaInputType=Bool
    alphaBoolBlend {blendOption=Linear}
    controlRigAssetReference {blueprintRigClass=/Game/Characters/Player/Male/Medium/Bodies/M_MED_Progressive_Jonesy_Kepler/Meshes/M_MED_Progressive_Jonesy_Kepler_CtrlRig/M_MED_Progressive_Jonesy_Kepler_CtrlRig.M_MED_Progressive_Jonesy_Kepler_CtrlRig_C}
   }
   n2 = Control Rig(Source=@n1, bAlphaBoolEnabled=NOT Boolean(A=Get bIsSkyDiving())) {
    alphaInputType=Bool
    alphaBoolBlend {blendOption=Linear}
    controlRigAssetReference {blueprintRigClass=&/M_MED_Progressive_Jonesy_Kepler_CtrlRig_target.M_MED_Progressive_Jonesy_Kepler_CtrlRig_target_C}
   }
   n3 = "Transform (Modify) Bone - Bone: dyn_shoulderpad_MSTR"!(
    ComponentPose=Local To Component(LocalPose=@n2)
    bAlphaBoolEnabled:bool=Get bIsSkyDiving()
    Translation="0,0,1"
    Rotation="0.000,0.000,60"
    Scale="1,1,1"
   ) -> (Pose) {
    alphaInputType=Bool
    rotationMode=BMM_Additive
    rotationSpace=BCS_BoneSpace
    translationMode=BMM_Additive
    translationSpace=BCS_BoneSpace
    alphaBoolBlend {blendInTime=0.1, blendOption=Linear, blendOutTime=0.1}
    alphaScaleBiasClamp {interpSpeedIncreasing=40}
    boneToModify {boneName=dyn_shoulderpad_MSTR}
    rotation {roll=60}
    translation {z=1}
   }
   n4 = "Transform (Modify) Bone - Bone: dyn_shoulderpad_MSTR"(ComponentPose=@n3, bAlphaBoolEnabled=Get bIsParachuteOpened(), Translation="0,-1,1", Rotation="0.000,0.000,-20", Scale="1,1,1") {
    alphaInputType=Bool
    rotationMode=BMM_Additive
    rotationSpace=BCS_BoneSpace
    translationMode=BMM_Additive
    translationSpace=BCS_BoneSpace
    alphaBoolBlend {blendInTime=0.1, blendOption=Linear, blendOutTime=0.1}
    alphaScaleBiasClamp {interpSpeedIncreasing=40}
    boneToModify {boneName=dyn_shoulderpad_MSTR}
    rotation {roll=-20}
    translation {y=-1, z=1}
   }
   n5 = "Transform (Modify) Bone - Bone: dyn_shoulderpad_MSTR"(ComponentPose=@n4, bAlphaBoolEnabled=Get bIsSkydiveDiveMode(), Translation="0,0,0", Rotation="0,0,-30", Scale="1,1,1") {
    alphaInputType=Bool
    rotationMode=BMM_Additive
    rotationSpace=BCS_BoneSpace
    translationMode=BMM_Additive
    translationSpace=BCS_BoneSpace
    alphaBoolBlend {blendInTime=0.1, blendOption=Linear, blendOutTime=0.1}
    alphaScaleBiasClamp {interpSpeedIncreasing=40}
    boneToModify {boneName=dyn_shoulderpad_MSTR}
    rotation {roll=-30}
   }
   n6 = "Transform (Modify) Bone - Bone: dyn_arm_blades"(ComponentPose=@n5, Alpha=1, Translation="0,0,0", Rotation="-1,0.000,0.000", Scale="1,1,1") {rotationMode=BMM_Additive, alphaBoolBlend {blendOption=Linear}, boneToModify {boneName=dyn_arm_blades}, rotation {pitch=-1}}
   n7 = "Transform (Modify) Bone - Bone: dyn_shoulderpad_MSTR"!(ComponentPose=@n6, Alpha:float=1, Translation="0,0,-0.5", Rotation="0,-3,10", Scale="1,1,1") -> (Pose) {
    rotationMode=BMM_Additive
    rotationSpace=BCS_BoneSpace
    translationMode=BMM_Additive
    translationSpace=BCS_BoneSpace
    alphaBoolBlend {blendOption=Linear}
    alphaScaleBiasClamp {interpSpeedIncreasing=40}
    boneToModify {boneName=dyn_shoulderpad_MSTR}
    rotation {roll=10, yaw=-3}
    translation {z=-0.5}
   }
   n8 = "Transform (Modify) Bone - Bone: dyn_shoulder_pad"(ComponentPose=@n7, Alpha=1, Translation="0,0,0.5", Rotation="0,0,0", Scale="1,1,1") {
    rotationMode=BMM_Additive
    rotationSpace=BCS_BoneSpace
    translationMode=BMM_Additive
    translationSpace=BCS_BoneSpace
    alphaBoolBlend {blendOption=Linear}
    alphaScaleBiasClamp {interpSpeedIncreasing=40}
    boneToModify {boneName=dyn_shoulder_pad}
    translation {z=0.5}
   }
   n9 = "Transform (Modify) Bone - Bone: dyn_shoulderpad_MSTR"!(ComponentPose=@n8, bAlphaBoolEnabled:bool=Get bIsTargeting(), Translation="0,0,0", Rotation="0,0,-20", Scale="1,1,1") -> (Pose) {
    alphaInputType=Bool
    rotationMode=BMM_Additive
    rotationSpace=BCS_BoneSpace
    scaleMode=BMM_Additive
    scaleSpace=BCS_BoneSpace
    translationMode=BMM_Additive
    translationSpace=BCS_BoneSpace
    alphaBoolBlend {blendInTime=0.1, blendOption=Linear, blendOutTime=0.1}
    boneToModify {boneName=dyn_shoulderpad_MSTR}
    rotation {roll=-20}
   }
   n10 = "Constraint - dyn_pouch_rt_MSTR_01 (PPPPPP)"!(
    ComponentPose=@n9
    Alpha=1
    ConstraintWeights_0=0.3474
    ConstraintWeights_1:float=0.2042
    ConstraintWeights_2:float=0.1286
    ConstraintWeights_3:float=0.0288
    ConstraintWeights_4:float=0.0003
    ConstraintWeights_5:float=0.2907
   ) -> (Pose) {
    alphaBoolBlend {blendOption=Linear}
    boneToModify {boneName=dyn_pouch_rt_MSTR_01}
    constraintSetup [
     {targetBone {boneName=spine_01}}
     {targetBone {boneName=spine_02}}
     {targetBone {boneName=spine_03}}
     {targetBone {boneName=spine_04}}
     {targetBone {boneName=spine_05}}
     {targetBone {boneName=pelvis}}
    ]
    constraintWeights [1,1,1,1,1,1]
   }
   n11 = "Constraint - dyn_pouch_rt_MSTR_02 (PPPPPP)"!(
    ComponentPose=@n10
    Alpha=1
    ConstraintWeights_0=0.4033
    ConstraintWeights_1:float=0.1874
    ConstraintWeights_2:float=0.0916
    ConstraintWeights_3:float=0.0299
    ConstraintWeights_4:float=0.0002
    ConstraintWeights_5:float=0.2875
   ) -> (Pose) {
    alphaBoolBlend {blendOption=Linear}
    boneToModify {boneName=dyn_pouch_rt_MSTR_02}
    constraintSetup [
     {targetBone {boneName=spine_01}}
     {targetBone {boneName=spine_02}}
     {targetBone {boneName=spine_03}}
     {targetBone {boneName=spine_04}}
     {targetBone {boneName=spine_05}}
     {targetBone {boneName=pelvis}}
    ]
    constraintWeights [1,1,1,1,1,1]
   }
   n12 = "Constraint - dyn_pouch_rt_MSTR_03 (PPPPPP)"!(
    ComponentPose=@n11
    Alpha=1
    ConstraintWeights_0=0.4115
    ConstraintWeights_1:float=0.1997
    ConstraintWeights_2:float=0.0909
    ConstraintWeights_3:float=0.0467
    ConstraintWeights_4:float=0.0002
    ConstraintWeights_5:float=0.2511
   ) -> (Pose) {
    alphaBoolBlend {blendOption=Linear}
    boneToModify {boneName=dyn_pouch_rt_MSTR_03}
    constraintSetup [
     {targetBone {boneName=spine_01}}
     {targetBone {boneName=spine_02}}
     {targetBone {boneName=spine_03}}
     {targetBone {boneName=spine_04}}
     {targetBone {boneName=spine_05}}
     {targetBone {boneName=pelvis}}
    ]
    constraintWeights [1,1,1,1,1,1]
   }
   n13 = "Constraint - dyn_pouch_bk_MSTR_01 (PPPPPP)"!(
    ComponentPose=@n12
    Alpha=1
    ConstraintWeights_0=0.3609
    ConstraintWeights_1:float=0.2567
    ConstraintWeights_2:float=0.151
    ConstraintWeights_3:float=0.0746
    ConstraintWeights_4:float=0.0017
    ConstraintWeights_5:float=0.1551
   ) -> (Pose) {
    alphaBoolBlend {blendOption=Linear}
    boneToModify {boneName=dyn_pouch_bk_MSTR_01}
    constraintSetup [
     {targetBone {boneName=spine_01}}
     {targetBone {boneName=spine_02}}
     {targetBone {boneName=spine_03}}
     {targetBone {boneName=spine_04}}
     {targetBone {boneName=spine_05}}
     {targetBone {boneName=pelvis}}
    ]
    constraintWeights [1,1,1,1,1,1]
   }
   n14 = "Constraint - dyn_pouch_bk_MSTR_02 (PPPPPP)"!(
    ComponentPose=@n13
    Alpha=1
    ConstraintWeights_0=0.3562
    ConstraintWeights_1:float=0.2572
    ConstraintWeights_2:float=0.1561
    ConstraintWeights_3:float=0.0804
    ConstraintWeights_4:float=0.0024
    ConstraintWeights_5:float=0.1477
   ) -> (Pose) {
    alphaBoolBlend {blendOption=Linear}
    boneToModify {boneName=dyn_pouch_bk_MSTR_02}
    constraintSetup [
     {targetBone {boneName=spine_01}}
     {targetBone {boneName=spine_02}}
     {targetBone {boneName=spine_03}}
     {targetBone {boneName=spine_04}}
     {targetBone {boneName=spine_05}}
     {targetBone {boneName=pelvis}}
    ]
    constraintWeights [1,1,1,1,1,1]
   }
   n15 = "Constraint - dyn_pouch_bk_MSTR_03 (PPPPPP)"!(
    ComponentPose=@n14
    Alpha=1
    ConstraintWeights_0=0.3587
    ConstraintWeights_1:float=0.252
    ConstraintWeights_2:float=0.1479
    ConstraintWeights_3:float=0.076
    ConstraintWeights_4:float=0.0013
    ConstraintWeights_5:float=0.1641
   ) -> (Pose) {
    alphaBoolBlend {blendOption=Linear}
    boneToModify {boneName=dyn_pouch_bk_MSTR_03}
    constraintSetup [
     {targetBone {boneName=spine_01}}
     {targetBone {boneName=spine_02}}
     {targetBone {boneName=spine_03}}
     {targetBone {boneName=spine_04}}
     {targetBone {boneName=spine_05}}
     {targetBone {boneName=pelvis}}
    ]
    constraintWeights [1,1,1,1,1,1]
   }
   n16 = "Constraint - dyn_pouch_lt_MSTR_01 (PPPPPP)"!(
    ComponentPose=@n15
    Alpha=1
    ConstraintWeights_0=0.3589
    ConstraintWeights_1:float=0.1969
    ConstraintWeights_2:float=0.1192
    ConstraintWeights_3:float=0.0239
    ConstraintWeights_4:float=0
    ConstraintWeights_5:float=0.301
   ) -> (Pose) {
    alphaBoolBlend {blendOption=Linear}
    boneToModify {boneName=dyn_pouch_lt_MSTR_01}
    constraintSetup [
     {targetBone {boneName=spine_01}}
     {targetBone {boneName=spine_02}}
     {targetBone {boneName=spine_03}}
     {targetBone {boneName=spine_04}}
     {targetBone {boneName=spine_05}}
     {targetBone {boneName=pelvis}}
    ]
    constraintWeights [1,1,1,1,1,1]
   }
   n17 = "Constraint - dyn_pouch_lt_MSTR_02 (PPPPPP)"!(
    ComponentPose=@n16
    Alpha=1
    ConstraintWeights_0=0.3853
    ConstraintWeights_1:float=0.1921
    ConstraintWeights_2:float=0.0999
    ConstraintWeights_3:float=0.0268
    ConstraintWeights_4:float=0
    ConstraintWeights_5:float=0.2959
   ) -> (Pose) {
    alphaBoolBlend {blendOption=Linear}
    boneToModify {boneName=dyn_pouch_lt_MSTR_02}
    constraintSetup [
     {targetBone {boneName=spine_01}}
     {targetBone {boneName=spine_02}}
     {targetBone {boneName=spine_03}}
     {targetBone {boneName=spine_04}}
     {targetBone {boneName=spine_05}}
     {targetBone {boneName=pelvis}}
    ]
    constraintWeights [1,1,1,1,1,1]
   }
   n18 = "Constraint - dyn_pouch_lt_MSTR_03 (PPPPPP)"!(
    ComponentPose=@n17
    Alpha=1
    ConstraintWeights_0=0.3969
    ConstraintWeights_1:float=0.2037
    ConstraintWeights_2:float=0.094
    ConstraintWeights_3:float=0.0437
    ConstraintWeights_4:float=0.0001
    ConstraintWeights_5:float=0.2616
   ) -> (Pose) {
    alphaBoolBlend {blendOption=Linear}
    boneToModify {boneName=dyn_pouch_lt_MSTR_03}
    constraintSetup [
     {targetBone {boneName=spine_01}}
     {targetBone {boneName=spine_02}}
     {targetBone {boneName=spine_03}}
     {targetBone {boneName=spine_04}}
     {targetBone {boneName=spine_05}}
     {targetBone {boneName=pelvis}}
    ]
    constraintWeights [1,1,1,1,1,1]
   }
   n19 = "Constraint - dyn_crossbow_01_l_MSTR (PP)"!(ComponentPose=@n18, Alpha=1, ConstraintWeights_0=0.75, ConstraintWeights_1:float=0.25) -> (Pose) {
    alphaBoolBlend {blendOption=Linear}
    boneToModify {boneName=dyn_crossbow_01_l_MSTR}
    constraintSetup [{targetBone {boneName=pelvis}}, {targetBone {boneName=thigh_l}}]
    constraintWeights [1,1]
   }
   n20 = "Constraint - dyn_shirt_01_r_MSTR (R)"(ComponentPose=@n19, Alpha=0, ConstraintWeights_0=1) {
    alphaBoolBlend {blendOption=Linear}
    boneToModify {boneName=dyn_shirt_01_r_MSTR}
    constraintSetup [{transformType=Rotation, targetBone {boneName=thigh_r}}]
    constraintWeights [1]
   }
   n21 = Get bIsCrouching()
   n22 = Get bIsTargeting()
   n23 = "Constraint - dyn_shirt_fr_01_r_MSTR (RR)"!(ComponentPose=@n20, bAlphaBoolEnabled:bool=NOT Boolean(A=OR Boolean(A=@n22, B=@n21)), ConstraintWeights_0=1, ConstraintWeights_1:float=0) -> (Pose) {
    alphaInputType=Bool
    alphaBoolBlend {blendInTime=0.1, blendOption=Linear, blendOutTime=0.1}
    boneToModify {boneName=dyn_shirt_fr_01_r_MSTR}
    constraintSetup [{transformType=Rotation, targetBone {boneName=thigh_twist_01_r}}, {transformType=Rotation, targetBone {boneName=thigh_r}}]
    constraintWeights [1,1]
   }
   n24 = "Constraint - dyn_shoulder_pad_MSTR (PPPPPPPP)"!(
    ComponentPose=@n23
    Alpha=1
    ConstraintWeights_0=0.6003
    ConstraintWeights_1:float=0.0421
    ConstraintWeights_2:float=0.0064
    ConstraintWeights_3:float=0.1164
    ConstraintWeights_4:float=0.2186
    ConstraintWeights_5:float=0.0125
    ConstraintWeights_6:float=0.0036
    ConstraintWeights_7:float=0.0001
   ) -> (Pose) {
    alphaBoolBlend {blendOption=Linear}
    boneToModify {boneName=dyn_shoulder_pad_MSTR}
    constraintSetup [
     {targetBone {boneName=clavicle_l}}
     {targetBone {boneName=spine_03}}
     {targetBone {boneName=spine_04}}
     {targetBone {boneName=spine_05}}
     {targetBone {boneName=upperarm_twist_01_l}}
     {targetBone {boneName=neck_02}}
     {targetBone {boneName=neck_01}}
     {targetBone {boneName=upperarm_l}}
    ]
    constraintWeights [1,1,1,1,1,1,1,1]
   }
   n25 = "Constraint - dyn_l_shoulder_MSTR (PPPPPPPP)"!(
    ComponentPose=@n24
    Alpha=1
    ConstraintWeights_0=0.6003
    ConstraintWeights_1:float=0.0421
    ConstraintWeights_2:float=0.0064
    ConstraintWeights_3:float=0.1164
    ConstraintWeights_4:float=0.2186
    ConstraintWeights_5:float=0.0125
    ConstraintWeights_6:float=0.0036
    ConstraintWeights_7:float=0.0001
   ) -> (Pose) {
    alphaBoolBlend {blendOption=Linear}
    boneToModify {boneName=dyn_l_shoulder_MSTR}
    constraintSetup [
     {targetBone {boneName=clavicle_l}}
     {targetBone {boneName=spine_03}}
     {targetBone {boneName=spine_04}}
     {targetBone {boneName=spine_05}}
     {targetBone {boneName=upperarm_twist_01_l}}
     {targetBone {boneName=neck_02}}
     {targetBone {boneName=neck_01}}
     {targetBone {boneName=upperarm_l}}
    ]
    constraintWeights [1,1,1,1,1,1,1,1]
   }
   n26 = "Constraint - dyn_cb_buckle_01_MSTR (PPPPPPP)"!(
    ComponentPose=@n25
    Alpha=1
    ConstraintWeights_0=0.3326
    ConstraintWeights_1:float=0.2047
    ConstraintWeights_2:float=0.1342
    ConstraintWeights_3:float=0.0277
    ConstraintWeights_4:float=0.0001
    ConstraintWeights_5:float=0.2984
    ConstraintWeights_6:float=0.0023
   ) -> (Pose) {
    alphaBoolBlend {blendOption=Linear}
    boneToModify {boneName=dyn_cb_buckle_01_MSTR}
    constraintSetup [
     {targetBone {boneName=spine_01}}
     {targetBone {boneName=spine_02}}
     {targetBone {boneName=spine_03}}
     {targetBone {boneName=spine_04}}
     {targetBone {boneName=spine_05}}
     {targetBone {boneName=pelvis}}
     {targetBone {boneName=thigh_twist_01_l}}
    ]
    constraintWeights [1,1,1,1,1,1,1]
   }
   n27 = "Transform (Modify) Bone - Bone: dyn_shirt_fr_01_r_MSTR"(ComponentPose=@n26, Alpha=1, Translation="-0.000,0,0.371", Rotation="0,0,0", Scale="1,1,1") {
    rotationMode=BMM_Additive
    rotationSpace=BCS_BoneSpace
    scaleMode=BMM_Additive
    scaleSpace=BCS_BoneSpace
    translationMode=BMM_Additive
    translationSpace=BCS_BoneSpace
    alphaBoolBlend {blendOption=Linear}
    boneToModify {boneName=dyn_shirt_fr_01_r_MSTR}
    translation {z=0.371}
   }
   n28 = "Transform (Modify) Bone - Bone: dyn_shirt_fr_01_r_MSTR"!(ComponentPose=@n27, bAlphaBoolEnabled:bool=Get bIsJumping(), Translation="0,0,0", Rotation="15,0,0", Scale="1,1,1") -> (Pose) {
    alphaInputType=Bool
    rotationMode=BMM_Additive
    rotationSpace=BCS_BoneSpace
    scaleMode=BMM_Additive
    scaleSpace=BCS_BoneSpace
    translationMode=BMM_Additive
    translationSpace=BCS_BoneSpace
    alphaBoolBlend {blendOption=Linear}
    boneToModify {boneName=dyn_shirt_fr_01_r_MSTR}
    rotation {pitch=15}
   }
   n29 = "Transform (Modify) Bone - Bone: dyn_crossbow_01_l"!(ComponentPose=@n28, bAlphaBoolEnabled:bool=@n21, Translation="-1.5,-1,-1", Rotation="0,0,0", Scale="1,1,1") -> (Pose) {
    alphaInputType=Bool
    rotationMode=BMM_Additive
    translationMode=BMM_Additive
    alphaBoolBlend {blendInTime=0.1, blendOption=Linear, blendOutTime=0.1}
    boneToModify {boneName=dyn_crossbow_01_l}
    translation=<-1.5 -1 -1>
   }
   n30 = "Transform (Modify) Bone - Bone: dyn_shirt_fr_01_r_MSTR"(ComponentPose=@n29, bAlphaBoolEnabled=@n21, Translation="-3,3,1", Rotation="0,0,0", Scale="1,1,1") {
    alphaInputType=Bool
    rotationMode=BMM_Additive
    rotationSpace=BCS_BoneSpace
    scaleMode=BMM_Additive
    scaleSpace=BCS_BoneSpace
    translationMode=BMM_Additive
    translationSpace=BCS_BoneSpace
    alphaBoolBlend {blendOption=Linear}
    boneToModify {boneName=dyn_shirt_fr_01_r_MSTR}
    translation=<-3 3 1>
   }
   n31 = "Transform (Modify) Bone - Bone: dyn_shirt_01_r_MSTR"!(ComponentPose=@n30, bAlphaBoolEnabled:bool=@n21, Translation="-3,3,1", Rotation="0,0,0", Scale="1,1,1") -> (Pose) {
    alphaInputType=Bool
    rotationMode=BMM_Additive
    rotationSpace=BCS_BoneSpace
    scaleMode=BMM_Additive
    scaleSpace=BCS_BoneSpace
    translationMode=BMM_Additive
    translationSpace=BCS_BoneSpace
    alphaBoolBlend {blendOption=Linear}
    boneToModify {boneName=dyn_shirt_01_r_MSTR}
    translation=<-3 3 1>
   }
   n32 = Get windVector() -> (windVector:Vector)
   n33 = RigidBody!(ComponentPose=@n31, Alpha, OverrideWorldGravity:Vector=@n32, ExternalForce) -> (Pose) {
    bClampLinearTranslationLimitToRefPose=true
    bOverrideWorldGravity=true
    overridePhysicsAsset=&/M_MED_Progressive_Jonesy_kepler_Cb_Physics
    simulationSpace=BaseBoneSpace
    alphaBoolBlend {blendOption=Linear}
    baseBoneRef {boneName=root}
    componentAppliedLinearAccClamp=<3000 3000 3000>
    componentLinearAccScale=<0.3 0.3 0.7>
    componentLinearVelScale=<0.3 0.3 0.7>
    overrideWorldGravity {z=-980}
   }
   n34 = "Transform (Modify) Bone - Bone: dyn_crossbow_03_l"!(ComponentPose=@n33, bAlphaBoolEnabled:bool=@n21, Translation="-0.5,-2,-0.5", Rotation="-5,0,0", Scale="1,1,1") -> (Pose) {
    alphaInputType=Bool
    translationMode=BMM_Additive
    alphaBoolBlend {blendInTime=0.1, blendOption=Linear, blendOutTime=0.1}
    boneToModify {boneName=dyn_crossbow_03_l}
    rotation {pitch=-5}
    translation=<-0.5 -2 -0.5>
   }
   n35 = RigidBody(ComponentPose=@n34, OverrideWorldGravity=@n32) {
    bClampLinearTranslationLimitToRefPose=true
    bOverrideWorldGravity=true
    overridePhysicsAsset=&/M_MED_Progressive_Jonesy_kepler_Physics
    simulationSpace=BaseBoneSpace
    alphaBoolBlend {blendOption=Linear}
    baseBoneRef {boneName=root}
    componentAppliedLinearAccClamp=<3000 3000 3000>
    componentLinearAccScale=<0.3 0.3 0.4>
    componentLinearVelScale=<0.3 0.3 0.4>
    overrideWorldGravity {z=-980}
   }
   Output Pose(Result=Component To Local(ComponentPose=@n35)) {layerGroup=None}
   Get pistonAimAlpha_R() -> (pistonAimAlpha_R:double)
  ]}
 EventGraph {nodes [
   Event BlueprintUpdateAnimation()
   > n1 = Sequence!(execute) -> (then_0, then_1, then_2:Exec)
   n2 = Break Fort Anim Input Skydiving External Force(FortAnimInput_SkydivingExternalForce=Get SkydivingExternalForceInput()) -> (
    bUseSkydivingVectorForce=~
    SkydivingVectorForce=~
    HeadToPelvisDirection=~
    FloatingMultiplier=~
    FloatingAdditive=~
    DivingMultiplier=~
    DivingAdditive=~
    ParachutingMultiplier=~
    ParachutingAdditive=~
   )
   > Set windVector(windVector:Vector=SelectVector(A=vector * float(A=@n2.SkydivingVectorForce, B=0.15), B="0, 0,-980", bPickA=Get bIsSkyDiving())) -> (Output_Get:Vector)
   n3 = Sequence!(execute=@n1.then_2) -> (then_0, then_1, then_2, then_3:Exec)
   n4 = GetOwnerSkeletalMeshComponent()
   n5 = GetDeltaTransformFromRefPose(self=@n4, BoneName=upperarm_r, BaseName=clavicle_r)
   Set uparmFwdBk_R(execute=@n3.then_1, uparmFwdBk_R:double=BreakRotator(InRot=BreakTransform(InTransform=@n5).Rotation).Yaw) -> (Output_Get:double)
   n6 = "Equal Exactly (String)"(A=Get weapon() -> (weapon:String), B=Pistol)
   n7 = Branch(execute=@n3.then_2, Condition=@n6)
   > Set pistonAimAlpha_R(pistonAimAlpha_R:double=NormalizeToRange!(Value:double=Get uparmFwdBk_R() -> (uparmFwdBk_R:double), RangeMin:double=50, RangeMax:double=120) -> (ReturnValue:double)) -> (Output_Get:double)
   Set pistonAimAlpha_R(execute=@n7.else, pistonAimAlpha_R=0)
   n8 = Get weapon()
   n9 = Branch(execute=@n3.then_3, Condition=NOT Boolean(A="Equal Exactly (String)"(A=@n8, B=pistol)))
   > Set pistonAimAlpha_R(pistonAimAlpha_R=NormalizeToRange(Value=Get uparmFwdBk_R(), RangeMin=30, RangeMax=60))
   Set pistonAimAlpha_R(execute=@n9.else, pistonAimAlpha_R=0)
   n10 = GetDeltaTransformFromRefPose(self=@n4, BoneName=upperarm_l, BaseName=clavicle_l)
   Set uparmFwdBk_L(execute=@n3.then_0, uparmFwdBk_L:double=BreakRotator(InRot=BreakTransform(InTransform=@n10).Rotation).Yaw) -> (Output_Get:double)
   n11 = Cast To FortPlayerAnimInstance(execute=@n1.then_1, Object=GetAnimInstance(self=GetOwnerSkeletalMeshComponent()))
   > Set weapon(weapon:String=Enum to String!(Enumerator:EFortWeaponCoreAnimation Enum=Get WeaponCoreAnim(self=@n11)) -> (ReturnValue:String)) -> (Output_Get:String)
   AND Boolean(A=false, B=false)
   AND Boolean(A=NOT Boolean(A=@n6), B=NOT Boolean(A=Get bIsSkyDiving()))
   Get bIsSkyDiving()
  ]}
]
variables [
 arm rotation_l {type=number}
 pistonAimAlpha_R {type=number}
 uparmFwdBk_L {type=number}
 uparmFwdBk_R {type=number}
 weapon {type=string, default=""}
 windVector {type=Vector}
]
