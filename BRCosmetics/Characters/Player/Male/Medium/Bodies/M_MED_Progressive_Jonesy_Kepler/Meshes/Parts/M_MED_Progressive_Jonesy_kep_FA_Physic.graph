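# Physics asset for the Kepler Progressive Jonesy face-accessory part, simulating its dynamic band bones against the head.
# Capsule bodies cover the head and the left and right band bones.
# Each band bone is constrained to the head with swing and twist angle limits, allowing bounded dangling motion.
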PhysicsAsset /BRCosmetics/Characters/Player/Male/Medium/Bodies/M_MED_Progressive_Jonesy_Kepler/Meshes/Parts/M_MED_Progressive_Jonesy_kep_FA_Physic
bodies {
 dyn_band_l [{shapeType=Capsule, length=1.1136, radius=1.402, center=<3 0 1>, rotation=<rot 90 0 90>}]
 dyn_band_r_01 [{shapeType=Capsule, length=1.1863, radius=0.9665, center=<2.1849 0 0>, rotation=<rot 0 0 -90>}]
 head [{shapeType=Capsule, length=4.1575, radius=6.3984, center=<2.8624 0.5728 0>, rotation=<rot 90 179.1048 174.1048>}]
}
constraints [
 {bone1Name=dyn_band_r_01, bone2Name=head, swing1LimitDegrees=45, swing2LimitDegrees=75, twistLimitDegrees=15}
 {bone1Name=dyn_band_l, bone2Name=head, swing1LimitDegrees=45, swing2LimitDegrees=75, twistLimitDegrees=15}
]
